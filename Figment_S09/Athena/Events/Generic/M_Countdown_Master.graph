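# Master unlit material for an in-world event countdown, showing timer digits, DHMS labels and a background, then a post-countdown image.
# MF_CountdownTimer builds the digits, tinted by a color or panning radial rainbow, and MPC_Countdown's phase 3 value blends to the optionally pulsing image.
# Vertex offset optionally billboards the mesh, or collapses it far underground when hidden by MPC or camera distance.
# Opacity adds vignette, distance fade and switchable hologram dot erosion.
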Material /Figment_S09/Athena/Events/Generic/M_Countdown_Master
graph {
 properties {bEnableResponsiveAA=true, pixelDepthOffsetMode=PDOM_Legacy, shadingModel=MSM_Unlit, usedShadingModels=Unlit}
 nodes [
  ActorPositionWS()
  n1 = MaterialFunctionCall() {materialFunction=/Game/ImpostorBaker/MaterialFunctions/ImpostorInfo}
  n2 = Add(A=VectorParameter() {parameterName=PositionOffsetHLOD, defaultValue=<rgba 0 0 0 1>}.RGB, B=@n1.Position)
  n3 = AppendVector(A=ScalarParameter() {defaultValue=200, parameterName=ScaleX}, B=ScalarParameter() {defaultValue=200, parameterName=ScaleY})
  n4 = Multiply(
   A=MaterialFunctionCall() {materialFunction=/Engine/Functions/Engine_MaterialFunctions02/WorldPositionOffset/ObjectScale}.Scale Z
   B=ComponentMask(None=TextureCoordinate() {coordinateIndex=3}) {g=true}
  )
  n5 = MaterialFunctionCall!(Center Location=@n2, 0-1 UVs, XY scale=Multiply(A=@n3, B=@n4), OverridePivot, "Normals (Optional)") {materialFunction=&Sprite}
  n6 = Subtract(A=CameraPositionWS(), B=ObjectPositionWS())
  n7 = ScalarParameter() {defaultValue=10117.7236}
  Add(A=@n5.World position offset, B=Multiply(A=Normalize(VectorInput=@n6), B=@n7))
  n8 = TextureCoordinate()
  Add(A=@n8, B=Constant2Vector() {r=-0.5})
  Divide(A=ComponentMask(None=@n1.Scaled Size) {b=true}, B=ComponentMask(None=@n1.Default Size) {b=true})
  n9 = TextureSampleParameter2D() {parameterName=Background, texture=/Game/Athena/Events/Generic/CountdownTimer/Textures/Countdown_Background_1}
  n10 = ScalarParameter() {defaultValue=-0.5, parameterName=xadd}
  n11 = ScalarParameter() {defaultValue=-2.7572, parameterName=yadd}
  n12 = Multiply(
   A=Add(A=@n8, B=AppendVector(A=@n10, B=@n11))
   B=AppendVector(A=ScalarParameter() {defaultValue=-1, parameterName=x}, B=ScalarParameter() {defaultValue=0.3448, parameterName=y})
  )
  n13 = Time()
  n14 = ScalarParameter() {defaultValue=1.1419, parameterName=SpeedRainbow}
  n15 = TextureSample(UVs=Panner(
    Coordinate=MaterialFunctionCall(Vector or UVs=@n12) {materialFunction=&VectorToRadialValue}.Radial Coordinates
    Time=Multiply(A=@n13, B=@n14)
   ) {speedX=1.667}) {texture=/Game/Global/Textures/Default/Masks_and_Gradients/T_Grad_Rainbow}
  n16 = Desaturation(
   None=MaterialFunctionCall(LDR Input=Clamp(None=@n15.RGB), HDR Tint=@n15.RGB) {materialFunction=&DeriveHDRfromLDR}.RemappedRange
   Fraction=ScalarParameter() {defaultValue=1, parameterName=Desat}
  )
  n17 = VectorParameter() {parameterName=Color, defaultValue=<rgba 1 1 1 1>}
  n18 = StaticSwitchParameter(True=Multiply(A=@n16, B=@n17.RGB), False=@n17.RGB) {parameterName=UseRainbowEffect}
  n19 = ScalarParameter() {defaultValue=1, parameterName=DHMS_Brightness}
  n20 = AppendVector(
   A=ScalarParameter() {group=Numbers, parameterName=u_DHMSPos}
   B=ScalarParameter() {defaultValue=-0.6, group=Numbers, parameterName=v_DHMSPos}
  )
  n21 = ScalarParameter() {group=Numbers, parameterName=u_NumberPaddingOffset}
  n22 = ScalarParameter() {defaultValue=-0.1524, group=Numbers, parameterName=v_NumberPaddingOffset}
  n23 = MaterialFunctionCall!(
   UVs=Add(A=TextureCoordinate(), B=AppendVector(A=@n21, B=@n22))
   UVPadding=AppendVector(
    A=ScalarParameter() {defaultValue=1, group=Numbers, parameterName=u_NumberPaddingScale}
    B=ScalarParameter() {defaultValue=0.4667, group=Numbers, parameterName=v_NumberPaddingScale}
   )
   "NumberofDigits(int)"=ScalarParameter() {defaultValue=11, parameterName=NumberofDigits}
  ) {materialFunction=/Game/Packages/Fortress_SharedMaterials/Base_Material_Functions/BaseCharacterShaderFunctions/MF_CountdownTimer}
  n24 = ScalarParameter() {defaultValue=1.0383, group=Numbers, parameterName=u_DHMSScale}
  n25 = ScalarParameter() {defaultValue=0.419, group=Numbers, parameterName=v_DHMSScale}
  n26 = TextureSampleParameter2D(UVs=MaterialFunctionCall(UVs=Add(A=@n20, B=@n23.UVS), Texture Scale=AppendVector(A=@n24, B=@n25)) {materialFunction=&ScaleUVsByCenter}.UVs) {parameterName=DHMS, samplerType=SAMPLERTYPE_Masks, texture=/Game/Athena/Playlists/Barrier/Textures/Countdown_DHMS}
  n27 = Add(A=@n9.A, B=Add(A=Multiply(A=@n19, B=@n26.R), B=@n23.NumbersGreyscale))
  n28 = AppendVector(
   A=ScalarParameter() {defaultValue=1, group=Numbers, parameterName=u_PostImagePaddingScale}
   B=ScalarParameter() {defaultValue=1, group=Numbers, parameterName=v_PostImagePaddingScale}
  )
  n29 = Frac(None=Multiply(A=Time(), B=ScalarParameter() {defaultValue=0.1, parameterName=SpeedMarsh}))
  n30 = Multiply(
   A=@n28
   B=LinearInterpolate(A=ScalarParameter() {defaultValue=0.5, parameterName=ScaleMin}, B=ScalarParameter() {defaultValue=1, parameterName=ScaleMax}, Alpha=@n29)
  )
  n31 = TextureSampleParameter2D(UVs=MaterialFunctionCall(Texture Scale=StaticSwitchParameter(True=@n28, False=@n30) {parameterName=NoAnimation}) {materialFunction=&ScaleUVsByCenter}.UVs) {parameterName=PostCountdownImage, texture=/Figment_S09/Characters/Fortress_Character/Textures/AllBlack}
  n32 = Multiply(A=Multiply(A=@n29, B=OneMinus(None=@n29))) {constB=2}
  n33 = StaticSwitchParameter(True=@n31.A, False=Multiply(A=@n31.A, B=Clamp(None=@n32))) {parameterName=NoAnimation}
  n34 = LinearInterpolate(A=Constant3Vector(), B=@n31.RGB, Alpha=@n33)
  n35 = StaticSwitchParameter(
   True=Constant() {r=1}
   False=CollectionParameter() {collection=/Game/Athena/Events/Generic/CountdownTimer/Materials/MPC_Countdown, parameterName=ShowCountdownPhase3}
  ) {parameterName=Debug2}
  LinearInterpolate(
   A=Add(A=Multiply(A=@n9.RGB, B=@n9.RGB), B=Multiply(A=@n18, B=@n27))
   B=StaticSwitchParameter(True=Multiply(A=@n34, B=@n16), False=@n34) {parameterName=UseHologramEffect}
   Alpha=LinearInterpolate(A=ScalarParameter() {defaultValue=0.02, parameterName=ShowPostImage}, B=Constant() {r=1}, Alpha=@n35)
  )
  n36 = ObjectPositionWS()
  n37 = WorldPosition()
  n38 = OneMinus(None=MaterialFunctionCall(
    Fade Length=ScalarParameter() {defaultValue=45000, parameterName=DistanceFadeLength}
    Fade Offset=ScalarParameter() {defaultValue=40000, parameterName=DistanceFadeOffset}
    For Vertex Shader=StaticBool() {value=true}
   ) {materialFunction=&CameraDepthFade}.Result)
  n39 = Ceil(None=@n38)
  n40 = Constant() {r=1}
  n41 = CollectionParameter() {collection=&MPC_Countdown, parameterName=ShowCountdown}
  n42 = StaticSwitchParameter(
   True=Multiply(
    A=@n41
    B=CollectionParameter() {collection=&MPC_Countdown, parameterName=ShowCountdownPhase2}
   )
   False=@n41
  ) {parameterName=isPhase2}
  LinearInterpolate(
   A=Add(A=Subtract(A=@n36, B=@n37.XYZ), B=Constant3Vector() {constant=<rgba 0 0 -5156156 0>})
   B=StaticSwitchParameter(True=@n5.World position offset, False=Constant3Vector()) {parameterName=IsCameraFacing}
   Alpha=Multiply(A=Saturate(None=@n39), B=StaticSwitchParameter(True=@n40, False=@n42) {parameterName=Debug})
  )
  MaterialFunctionCall() {materialFunction=&ObjectScale}
  n43 = Divide(A=Distance(A=@n2, B=CameraPositionWS()), B=ScalarParameter() {defaultValue=2117.0803, parameterName=Bias})
  n44 = ScalarParameter() {defaultValue=0.5, parameterName=ScaleMin}
  n45 = ScalarParameter() {defaultValue=1, parameterName=ScaleMax}
  Multiply(
   A=Clamp(None=Multiply(A=@n3, B=@n43), Min=Multiply(A=@n3, B=@n44), Max=Multiply(A=@n3, B=@n45))
   B=Multiply(A=ScalarParameter() {defaultValue=1, parameterName=Scale_Master}, B=@n4)
  )
  Multiply(A=@n7, B=@n4)
  n46 = Subtract(A=SceneDepth().R, B=ScalarParameter() {defaultValue=47881.7031, parameterName=ScenDepthFadeDistance})
  n47 = ScalarParameter() {defaultValue=2861.6772, parameterName=ScenDepthFade}
  Multiply(A=Multiply(A=@n27, B=ScalarParameter() {defaultValue=1, parameterName=Fadein}), B=Clamp(None=Divide(A=@n46, B=@n47)))
  n48 = ScalarParameter() {defaultValue=0.4307, parameterName=Erode}
  n49 = Divide(A=ScreenPosition().PixelPosition) {constB=5}
  n50 = TextureCoordinate()
  n51 = Constant2Vector() {g=0.5, r=0.5}
  n52 = Constant() {r=0.5}
  n53 = Constant()
  n54 = Add(
   A=LinearInterpolate(A=@n27, B=Multiply(A=@n31.R, B=@n33), Alpha=@n35)
   B=Multiply(A=SphereMask(A=@n50, B=@n51, Radius=@n52, Hardness=@n53), B=ScalarParameter() {defaultValue=0.5, parameterName=Vignette})
  )
  n55 = Multiply(
   A=SphereMask(A=Frac(None=@n49), B=Constant() {r=0.5}, Radius=ScalarParameter() {defaultValue=1.0191, parameterName=Param_1}, Hardness=Constant())
   B=@n54
  )
  Multiply(
   A=StaticSwitchParameter(
    True=MaterialFunctionCall!(Dynamic Param=@n48, Alpha=@n55) {materialFunction=/Game/Effects/Fort_Effects/Materials/_Functions/Erode_Alpha}.Result
    False=@n54
   ) {parameterName=UseHologramEffect}
   B=@n38
  )
  TextureSample() {texture=/Game/Athena/Events/Generic/CountdownTimer/Textures/Countdown_Background_BlackNoAlpha}
 ]
}
